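# Animation Blueprint for the long-shorts male body cosmetic, adding secondary motion to a pose copied from the attached parent mesh.
# Pouches get constraints and crouch-driven offsets, then AnimDynamics simulates pouches, pulls, collar and hood, with a look-at on the center collar.
# Shorts dynamics chains feed additive bone corrections weighted by knee bend, which the EventGraph derives from calf-to-thigh rotation.
# An unconnected Actor-space shorts chain with Flap nodes remains unused.
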AnimBlueprint /BRCosmetics/Characters/Player/Male/Medium/Bodies/M_MED_Long_Shorts/Meshes/M_MED_Long_Shorts_AnimBP
target_skeleton=&/M_MED_Long_Shorts_Skeleton
graphs [
 AnimGraph {nodes [
   n1 = Get calf_rot_back_r() -> (calf_rot_back_r:double)
   n2 = "Constraint - pouch_1_r (P)"!(
    ComponentPose=Local To Component(LocalPose=Copy Pose From Mesh() {bUseAttachedParent=true})
    Alpha=1
    ConstraintWeights_0=1
    ConstraintWeights_1:float=0
   ) -> (Pose) {
    alphaBoolBlend {blendOption=Linear}
    boneToModify {boneName=pouch_1_r}
    constraintSetup [{targetBone {boneName=pouch_1_r}}]
    constraintWeights [1,0]
   }
   n3 = "Constraint - pouch_2_r (P)"!(ComponentPose=@n2, Alpha=1, ConstraintWeights_0=1, ConstraintWeights_1:float=0) -> (Pose) {
    alphaBoolBlend {blendOption=Linear}
    boneToModify {boneName=pouch_2_r}
    constraintSetup [{targetBone {boneName=pouch_2_r}}]
    constraintWeights [1,0]
   }
   n4 = "Constraint - pouch_l (P)"!(ComponentPose=@n3, Alpha=1, ConstraintWeights_0=1, ConstraintWeights_1:float=0) -> (Pose) {
    alphaBoolBlend {blendOption=Linear}
    boneToModify {boneName=pouch_l}
    constraintSetup [{targetBone {boneName=pouch_l}}]
    constraintWeights [1,0]
   }
   n5 = "Constraint - pouch_3_r (P)"!(ComponentPose=@n4, Alpha=1, ConstraintWeights_0=1, ConstraintWeights_1:float=0) -> (Pose) {
    alphaBoolBlend {blendOption=Linear}
    boneToModify {boneName=pouch_3_r}
    constraintSetup [{targetBone {boneName=pouch_3_r}}]
    constraintWeights [1,0]
   }
   n6 = Get crouch_alpha() -> (crouch_alpha:double)
   n7 = "Transform (Modify) Bone - Bone: pouch_1_r"(ComponentPose=@n5, Alpha=@n6, Translation="0,2,7", Rotation="0,15,15", Scale="1,1,1") {
    rotationMode=BMM_Additive
    scaleMode=BMM_Additive
    translationMode=BMM_Additive
    translationSpace=BCS_ParentBoneSpace
    alphaBoolBlend {blendOption=Linear}
    alphaScaleBiasClamp {bInterpResult=true}
    boneToModify {boneName=pouch_1_r}
    rotation {roll=15, yaw=15}
    translation {y=2, z=7}
   }
   n8 = "Transform (Modify) Bone - Bone: pouch_2_r"(ComponentPose=@n7, Alpha=@n6, Translation="0,0,6", Rotation="0,15,6", Scale="1,1,1") {
    rotationMode=BMM_Additive
    scaleMode=BMM_Additive
    translationMode=BMM_Additive
    translationSpace=BCS_ParentBoneSpace
    alphaBoolBlend {blendOption=Linear}
    alphaScaleBiasClamp {bInterpResult=true}
    boneToModify {boneName=pouch_2_r}
    rotation {roll=6, yaw=15}
    translation {z=6}
   }
   n9 = "Transform (Modify) Bone - Bone: pouch_l"(ComponentPose=@n8, Alpha=@n6, Translation="0,0,-3", Rotation="0,-15,15", Scale="1,1,1") {
    rotationMode=BMM_Additive
    scaleMode=BMM_Additive
    translationMode=BMM_Additive
    translationSpace=BCS_ParentBoneSpace
    alphaBoolBlend {blendOption=Linear}
    alphaScaleBiasClamp {bInterpResult=true}
    boneToModify {boneName=pouch_l}
    rotation {roll=15, yaw=-15}
    translation {z=-3}
   }
   n10 = Anim Dynamics - Bone: pouch_1_r(ComponentPose=@n9, Alpha=1, ExternalForce="0,0,0") {
    angularDampingOverride=1
    angularSpringConstant=1500
    bAngularSpring=true
    bGravityOverrideInSimSpace=true
    bOverrideAngularDamping=true
    gravityScale=0.35
    simulationSpace=RootRelative
    alphaBoolBlend {blendOption=Linear}
    alphaScaleBiasClamp {scale=0.3}
    boundBone {boneName=pouch_1_r}
    physicsBodyDefinitions [{
      sphereCollisionRadius=0
      boundBone {boneName=pouch_1_r}
      boxExtents=<10 20 20>
      constraintSetup {angularLimitsMax=<0 0 0>, angularLimitsMin=<0 0 -40>, angularTarget=<0 0 0>}
      localJointOffset=<10 0 0>
     }]
   }
   n11 = Anim Dynamics - Bone: pouch_2_r(ComponentPose=@n10, Alpha=1, ExternalForce="0,0,0") {
    angularDampingOverride=1
    angularSpringConstant=1500
    bAngularSpring=true
    bGravityOverrideInSimSpace=true
    bOverrideAngularDamping=true
    gravityScale=0.35
    simulationSpace=RootRelative
    alphaBoolBlend {blendOption=Linear}
    alphaScaleBiasClamp {scale=0.3}
    boundBone {boneName=pouch_2_r}
    physicsBodyDefinitions [{
      sphereCollisionRadius=0
      boundBone {boneName=pouch_2_r}
      boxExtents=<10 20 20>
      constraintSetup {angularLimitsMax=<0 0 0>, angularLimitsMin=<0 0 -40>, angularTarget=<0 0 0>}
      localJointOffset=<10 0 0>
     }]
   }
   n12 = Anim Dynamics - Bone: pouch_l(ComponentPose=@n11, Alpha=1, ExternalForce="0,0,0") {
    angularDampingOverride=1
    angularSpringConstant=1500
    bAngularSpring=true
    bGravityOverrideInSimSpace=true
    bOverrideAngularDamping=true
    gravityScale=0.35
    simulationSpace=RootRelative
    alphaBoolBlend {blendOption=Linear}
    alphaScaleBiasClamp {scale=0.3}
    boundBone {boneName=pouch_l}
    physicsBodyDefinitions [{
      sphereCollisionRadius=0
      boundBone {boneName=pouch_l}
      boxExtents=<10 20 20>
      constraintSetup {angularLimitsMax=<0 0 0>, angularLimitsMin=<0 0 -40>, angularTarget=<0 0 0>}
      localJointOffset=<10 0 0>
     }]
   }
   n13 = Anim Dynamics - Bone: pouch_3_r(ComponentPose=@n12, Alpha=1, ExternalForce="0,0,0") {
    angularDampingOverride=1
    angularSpringConstant=1500
    bAngularSpring=true
    bGravityOverrideInSimSpace=true
    bOverrideAngularDamping=true
    gravityScale=0.35
    simulationSpace=RootRelative
    alphaBoolBlend {blendOption=Linear}
    alphaScaleBiasClamp {scale=0.3}
    boundBone {boneName=pouch_3_r}
    physicsBodyDefinitions [{
      sphereCollisionRadius=0
      boundBone {boneName=pouch_3_r}
      boxExtents=<10 20 20>
      constraintSetup {angularLimitsMax=<0 0 0>, angularLimitsMin=<0 0 -40>, angularTarget=<0 0 0>}
      localJointOffset=<10 0 0>
     }]
   }
   n14 = Anim Dynamics - Bone: pull_1_r(ComponentPose=@n13, Alpha=1, ExternalForce="0,0,0") {
    angularDampingOverride=1
    angularSpringConstant=1500
    bAngularSpring=true
    bGravityOverrideInSimSpace=true
    bOverrideAngularDamping=true
    bUseSphericalLimits=true
    gravityScale=0.35
    simulationSpace=RootRelative
    alphaBoolBlend {blendOption=Linear}
    alphaScaleBiasClamp {scale=0.8}
    boundBone {boneName=pull_1_r}
    componentLinearAccScale {z=0.4}
    componentLinearVelScale {z=0.4}
    physicsBodyDefinitions [{
      collisionType=CustomSphere
      sphereCollisionRadius=2
      boundBone {boneName=pull_1_r}
      boxExtents=<10 20 20>
      constraintSetup {angularLimitsMax=<5 30 30>, angularLimitsMin=<-5 -30 -40>, angularTarget=<0 0 0>}
      localJointOffset=<15 0 0>
     }]
    sphericalLimits [
     {limitRadius=15, drivingBone {boneName=spine_04}, sphereLocalOffset=<0 0 0>}
     {limitRadius=15, drivingBone {boneName=pelvis}, sphereLocalOffset=<10 0 0>}
    ]
   }
   n15 = Anim Dynamics - Bone: pull_2_r(ComponentPose=@n14, Alpha=1, ExternalForce="0,0,0") {
    angularDampingOverride=1
    angularSpringConstant=1500
    bAngularSpring=true
    bGravityOverrideInSimSpace=true
    bOverrideAngularDamping=true
    bUseSphericalLimits=true
    gravityScale=0.35
    simulationSpace=RootRelative
    alphaBoolBlend {blendOption=Linear}
    alphaScaleBiasClamp {scale=0.8}
    boundBone {boneName=pull_2_r}
    componentLinearAccScale {z=0.43}
    componentLinearVelScale {z=0.43}
    physicsBodyDefinitions [{
      collisionType=CustomSphere
      sphereCollisionRadius=2
      boundBone {boneName=pull_2_r}
      boxExtents=<10 20 20>
      constraintSetup {angularLimitsMax=<5 30 30>, angularLimitsMin=<-5 -30 -40>, angularTarget=<0 0 0>}
      localJointOffset=<15 0 0>
     }]
    sphericalLimits [
     {limitRadius=15, drivingBone {boneName=spine_04}, sphereLocalOffset=<0 0 0>}
     {limitRadius=15, drivingBone {boneName=pelvis}, sphereLocalOffset=<10 0 0>}
    ]
   }
   n16 = Anim Dynamics - Bone: pull_3_r(ComponentPose=@n15, Alpha=1, ExternalForce="0,0,0") {
    angularDampingOverride=1
    angularSpringConstant=1500
    bAngularSpring=true
    bGravityOverrideInSimSpace=true
    bOverrideAngularDamping=true
    bUseSphericalLimits=true
    gravityScale=0.35
    simulationSpace=RootRelative
    alphaBoolBlend {blendOption=Linear}
    alphaScaleBiasClamp {scale=0.8}
    boundBone {boneName=pull_3_r}
    componentLinearAccScale {z=0.46}
    componentLinearVelScale {z=0.46}
    physicsBodyDefinitions [{
      collisionType=CustomSphere
      sphereCollisionRadius=1
      boundBone {boneName=pull_3_r}
      boxExtents=<10 20 20>
      constraintSetup {angularLimitsMax=<5 30 30>, angularLimitsMin=<-5 -30 -40>, angularTarget=<0 0 0>}
      localJointOffset=<15 0 0>
     }]
    sphericalLimits [
     {limitRadius=15, drivingBone {boneName=spine_04}, sphereLocalOffset=<0 0 0>}
     {limitRadius=15, drivingBone {boneName=pelvis}, sphereLocalOffset=<10 0 0>}
    ]
   }
   n17 = Anim Dynamics - Bone: pull_1_l(ComponentPose=@n16, Alpha=1, ExternalForce="0,0,0") {
    angularDampingOverride=1
    angularSpringConstant=1500
    bAngularSpring=true
    bGravityOverrideInSimSpace=true
    bOverrideAngularDamping=true
    bUseSphericalLimits=true
    gravityScale=0.35
    simulationSpace=RootRelative
    alphaBoolBlend {blendOption=Linear}
    alphaScaleBiasClamp {scale=0.8}
    boundBone {boneName=pull_1_l}
    componentLinearAccScale {z=0.4}
    componentLinearVelScale {z=0.4}
    physicsBodyDefinitions [{
      collisionType=CustomSphere
      sphereCollisionRadius=2
      boundBone {boneName=pull_1_l}
      boxExtents=<10 20 20>
      constraintSetup {angularLimitsMax=<5 30 30>, angularLimitsMin=<-5 -30 -40>, angularTarget=<0 0 0>}
      localJointOffset=<15 0 0>
     }]
    sphericalLimits [
     {limitRadius=15, drivingBone {boneName=spine_04}, sphereLocalOffset=<0 0 0>}
     {limitRadius=15, drivingBone {boneName=pelvis}, sphereLocalOffset=<10 0 0>}
    ]
   }
   n18 = Anim Dynamics - Bone: pull_2_l(ComponentPose=@n17, Alpha=1, ExternalForce="0,0,0") {
    angularDampingOverride=1
    angularSpringConstant=1500
    bAngularSpring=true
    bGravityOverrideInSimSpace=true
    bOverrideAngularDamping=true
    bUseSphericalLimits=true
    gravityScale=0.35
    simulationSpace=RootRelative
    alphaBoolBlend {blendOption=Linear}
    alphaScaleBiasClamp {scale=0.8}
    boundBone {boneName=pull_2_l}
    componentLinearAccScale {z=0.43}
    componentLinearVelScale {z=0.43}
    physicsBodyDefinitions [{
      collisionType=CustomSphere
      sphereCollisionRadius=2
      boundBone {boneName=pull_2_l}
      boxExtents=<10 20 20>
      constraintSetup {angularLimitsMax=<5 30 30>, angularLimitsMin=<-5 -30 -40>, angularTarget=<0 0 0>}
      localJointOffset=<15 0 0>
     }]
    sphericalLimits [
     {limitRadius=15, drivingBone {boneName=spine_04}, sphereLocalOffset=<0 0 0>}
     {limitRadius=15, drivingBone {boneName=pelvis}, sphereLocalOffset=<10 0 0>}
    ]
   }
   n19 = Anim Dynamics - Bone: pull_3_l(ComponentPose=@n18, Alpha=1, ExternalForce="0,0,0") {
    angularDampingOverride=1
    angularSpringConstant=1500
    bAngularSpring=true
    bGravityOverrideInSimSpace=true
    bOverrideAngularDamping=true
    bUseSphericalLimits=true
    gravityScale=0.35
    simulationSpace=RootRelative
    alphaBoolBlend {blendOption=Linear}
    alphaScaleBiasClamp {scale=0.8}
    boundBone {boneName=pull_3_l}
    componentLinearAccScale {z=0.46}
    componentLinearVelScale {z=0.46}
    physicsBodyDefinitions [{
      collisionType=CustomSphere
      sphereCollisionRadius=1
      boundBone {boneName=pull_3_l}
      boxExtents=<10 20 20>
      constraintSetup {angularLimitsMax=<5 30 30>, angularLimitsMin=<-5 -30 -40>, angularTarget=<0 0 0>}
      localJointOffset=<15 0 0>
     }]
    sphericalLimits [
     {limitRadius=15, drivingBone {boneName=spine_04}, sphereLocalOffset=<0 0 0>}
     {limitRadius=15, drivingBone {boneName=pelvis}, sphereLocalOffset=<10 0 0>}
    ]
   }
   n20 = Look At - Bone: collar_c(ComponentPose=@n19, Alpha=1) {
    alphaBoolBlend {blendOption=Linear}
    alphaScaleBiasClamp {scale=0.8}
    boneToModify {boneName=collar_c}
    lookAtLocation {x=0}
    lookAtTarget {boneReference {boneName=collar_c_tgt}}
    lookAt_Axis {axis=<-1 0>}
    lookUp_Axis {axis=<0 1>}
   }
   n21 = Anim Dynamics - Bone: collar_c(ComponentPose=@n20, Alpha=1, ExternalForce="0,0,0") {
    angularDampingOverride=1
    angularSpringConstant=1500
    bAngularSpring=true
    bGravityOverrideInSimSpace=true
    bOverrideAngularDamping=true
    gravityScale=0.35
    simulationSpace=RootRelative
    alphaBoolBlend {blendOption=Linear}
    alphaScaleBiasClamp {scale=0.3}
    boundBone {boneName=collar_c}
    physicsBodyDefinitions [{
      sphereCollisionRadius=0
      boundBone {boneName=collar_c}
      boxExtents=<10 20 20>
      constraintSetup {angularLimitsMax=<0 0 0>, angularLimitsMin=<0 0 -40>, angularTarget=<0 0 0>}
      localJointOffset=<10 0 0>
     }]
   }
   n22 = Anim Dynamics - Bone: collar_r(ComponentPose=@n21, Alpha=1, ExternalForce="0,0,0") {
    angularDampingOverride=1
    angularSpringConstant=1500
    bAngularSpring=true
    bGravityOverrideInSimSpace=true
    bOverrideAngularDamping=true
    gravityScale=0.35
    simulationSpace=RootRelative
    alphaBoolBlend {blendOption=Linear}
    alphaScaleBiasClamp {scale=0.3}
    boundBone {boneName=collar_r}
    physicsBodyDefinitions [{
      sphereCollisionRadius=0
      boundBone {boneName=collar_r}
      boxExtents=<10 20 20>
      constraintSetup {angularLimitsMax=<0 0 0>, angularLimitsMin=<0 0 -40>, angularTarget=<0 0 0>}
      localJointOffset=<10 0 0>
     }]
   }
   n23 = Anim Dynamics - Bone: collar_l(ComponentPose=@n22, Alpha=1, ExternalForce="0,0,0") {
    angularDampingOverride=1
    angularSpringConstant=1500
    bAngularSpring=true
    bGravityOverrideInSimSpace=true
    bOverrideAngularDamping=true
    gravityScale=0.35
    simulationSpace=RootRelative
    alphaBoolBlend {blendOption=Linear}
    alphaScaleBiasClamp {scale=0.3}
    boundBone {boneName=collar_l}
    physicsBodyDefinitions [{
      sphereCollisionRadius=0
      boundBone {boneName=collar_l}
      boxExtents=<10 20 20>
      constraintSetup {angularLimitsMax=<0 0 0>, angularLimitsMin=<0 0 -40>, angularTarget=<0 0 0>}
      localJointOffset=<10 0 0>
     }]
   }
   n24 = Anim Dynamics - Bone: collar_side_r(ComponentPose=@n23, Alpha=1, ExternalForce="0,0,0") {
    angularDampingOverride=1
    angularSpringConstant=1500
    bAngularSpring=true
    bGravityOverrideInSimSpace=true
    bOverrideAngularDamping=true
    gravityScale=0.35
    simulationSpace=RootRelative
    alphaBoolBlend {blendOption=Linear}
    alphaScaleBiasClamp {scale=0.3}
    boundBone {boneName=collar_side_r}
    physicsBodyDefinitions [{
      sphereCollisionRadius=0
      boundBone {boneName=collar_side_r}
      boxExtents=<10 20 20>
      constraintSetup {angularLimitsMax=<0 0 0>, angularLimitsMin=<0 0 -40>, angularTarget=<0 0 0>}
      localJointOffset=<10 0 0>
     }]
   }
   n25 = Anim Dynamics - Bone: collar_side_l(ComponentPose=@n24, Alpha=1, ExternalForce="0,0,0") {
    angularDampingOverride=1
    angularSpringConstant=1500
    bAngularSpring=true
    bGravityOverrideInSimSpace=true
    bOverrideAngularDamping=true
    gravityScale=0.35
    simulationSpace=RootRelative
    alphaBoolBlend {blendOption=Linear}
    alphaScaleBiasClamp {scale=0.3}
    boundBone {boneName=collar_side_l}
    physicsBodyDefinitions [{
      sphereCollisionRadius=0
      boundBone {boneName=collar_side_l}
      boxExtents=<10 20 20>
      constraintSetup {angularLimitsMax=<0 0 0>, angularLimitsMin=<0 0 -40>, angularTarget=<0 0 0>}
      localJointOffset=<10 0 0>
     }]
   }
   n26 = Anim Dynamics - Bone: hood_c(ComponentPose=@n25, Alpha=1, ExternalForce="0,0,0") {
    angularDampingOverride=1
    angularSpringConstant=1500
    bAngularSpring=true
    bGravityOverrideInSimSpace=true
    bOverrideAngularDamping=true
    gravityScale=0.35
    simulationSpace=RootRelative
    alphaBoolBlend {blendOption=Linear}
    alphaScaleBiasClamp {scale=0.5}
    boundBone {boneName=hood_c}
    componentAppliedLinearAccClamp {z=5000}
    componentLinearAccScale {z=0.5}
    componentLinearVelScale {z=0.5}
    physicsBodyDefinitions [{
      sphereCollisionRadius=0
      boundBone {boneName=hood_c}
      boxExtents=<10 20 20>
      constraintSetup {angularLimitsMax=<0 45 0>, angularLimitsMin=<0 -45 -90>, angularTarget=<0 0 0>}
      localJointOffset=<10 0 0>
     }]
   }
   n27 = Get Shorts_Alpha() -> (Shorts_Alpha:double)
   n28 = "Anim Dynamics - Chain: dyn_shorts_in_l -> dyn_shorts_in_02_l"(ComponentPose=@n26, Alpha=@n27, ExternalForce="0,0,0") {
    angularDampingOverride=0.999
    angularSpringConstant=100
    bAngularSpring=true
    bChain=true
    bGravityOverrideInSimSpace=true
    bOverrideAngularDamping=true
    numSolverIterationsPostUpdate=4
    numSolverIterationsPreUpdate=16
    simulationSpace=RootRelative
    alphaBoolBlend {blendOption=Linear}
    boundBone {boneName=dyn_shorts_in_l}
    chainEnd {boneName=dyn_shorts_in_02_l}
    physicsBodyDefinitions [{
      sphereCollisionRadius=0
      boundBone {boneName=dyn_shorts_in_l}
      boxExtents=<25 30 12>
      constraintSetup {angularTargetAxis=AxisZ, twistAxis=AxisZ, angularLimitsMax=<10 10 5>, angularLimitsMin=<-10 -10 -5>, angularTarget=<0 0 1>}
      localJointOffset=<0 0 -3.5>
     }]
   }
   n29 = "Anim Dynamics - Chain: dyn_shorts_out_l -> dyn_shorts_out_02_l"(ComponentPose=@n28, Alpha=@n27, ExternalForce="0,0,0") {
    angularDampingOverride=0.999
    angularSpringConstant=100
    bAngularSpring=true
    bChain=true
    bGravityOverrideInSimSpace=true
    bOverrideAngularDamping=true
    numSolverIterationsPostUpdate=4
    numSolverIterationsPreUpdate=16
    simulationSpace=RootRelative
    alphaBoolBlend {blendOption=Linear}
    boundBone {boneName=dyn_shorts_out_l}
    chainEnd {boneName=dyn_shorts_out_02_l}
    physicsBodyDefinitions [{
      sphereCollisionRadius=0
      boundBone {boneName=dyn_shorts_out_l}
      boxExtents=<50 50 30>
      constraintSetup {angularTargetAxis=AxisZ, twistAxis=AxisZ, angularLimitsMax=<10 10 5>, angularLimitsMin=<-10 -10 -5>, angularTarget=<0 0 1>}
      localJointOffset=<0 0 -6.5>
     }]
   }
   n30 = "Anim Dynamics - Chain: dyn_shorts_fr_l -> dyn_shorts_fr_02_l"(ComponentPose=@n29, Alpha=@n27, ExternalForce="0,0,0") {
    angularDampingOverride=0.999
    angularSpringConstant=300
    bAngularSpring=true
    bChain=true
    bGravityOverrideInSimSpace=true
    bOverrideAngularDamping=true
    numSolverIterationsPostUpdate=4
    numSolverIterationsPreUpdate=16
    simulationSpace=RootRelative
    alphaBoolBlend {blendOption=Linear}
    boundBone {boneName=dyn_shorts_fr_l}
    chainEnd {boneName=dyn_shorts_fr_02_l}
    physicsBodyDefinitions [{
      sphereCollisionRadius=0
      boundBone {boneName=dyn_shorts_fr_l}
      boxExtents=<18 25 8>
      constraintSetup {angularTargetAxis=AxisZ, twistAxis=AxisZ, angularLimitsMax=<10 0 5>, angularLimitsMin=<-10 -20 -5>, angularTarget=<0 0 1>}
      localJointOffset=<0 0 -5>
     }]
   }
   n31 = "Anim Dynamics - Chain: dyn_shorts_bk_l -> dyn_shorts_bk_02_l"(ComponentPose=@n30, Alpha=@n27, ExternalForce="0,0,0") {
    angularDampingOverride=0.999
    angularSpringConstant=300
    bAngularSpring=true
    bChain=true
    bGravityOverrideInSimSpace=true
    bOverrideAngularDamping=true
    numSolverIterationsPostUpdate=4
    numSolverIterationsPreUpdate=16
    simulationSpace=RootRelative
    alphaBoolBlend {blendOption=Linear}
    boundBone {boneName=dyn_shorts_bk_l}
    chainEnd {boneName=dyn_shorts_bk_02_l}
    physicsBodyDefinitions [{
      sphereCollisionRadius=0
      boundBone {boneName=dyn_shorts_bk_l}
      boxExtents=<40 38 15>
      constraintSetup {angularTargetAxis=AxisZ, twistAxis=AxisZ, angularLimitsMax=<10 0 5>, angularLimitsMin=<-10 -20 -5>, angularTarget=<0 0 1>}
      localJointOffset=<0 0 -3.5>
     }]
   }
   n32 = "Anim Dynamics - Chain: dyn_shorts_in_r -> dyn_shorts_in_02_r"(ComponentPose=@n31, Alpha=@n27, ExternalForce="0,0,0") {
    angularDampingOverride=0.999
    angularSpringConstant=100
    bAngularSpring=true
    bChain=true
    bGravityOverrideInSimSpace=true
    bOverrideAngularDamping=true
    numSolverIterationsPostUpdate=4
    numSolverIterationsPreUpdate=16
    simulationSpace=RootRelative
    alphaBoolBlend {blendOption=Linear}
    boundBone {boneName=dyn_shorts_in_r}
    chainEnd {boneName=dyn_shorts_in_02_r}
    physicsBodyDefinitions [{
      sphereCollisionRadius=0
      boundBone {boneName=dyn_shorts_in_r}
      boxExtents=<25 29 12>
      constraintSetup {angularTargetAxis=AxisZ, twistAxis=AxisZ, angularLimitsMax=<10 10 5>, angularLimitsMin=<-10 -10 -5>, angularTarget=<0 0 1>}
      localJointOffset=<0 0 3.5>
     }]
   }
   n33 = "Anim Dynamics - Chain: dyn_shorts_out_r -> dyn_shorts_out_02_r"(ComponentPose=@n32, Alpha=@n27, ExternalForce="0,0,0") {
    angularDampingOverride=0.999
    angularSpringConstant=100
    bAngularSpring=true
    bChain=true
    bGravityOverrideInSimSpace=true
    bOverrideAngularDamping=true
    numSolverIterationsPostUpdate=4
    numSolverIterationsPreUpdate=16
    simulationSpace=RootRelative
    alphaBoolBlend {blendOption=Linear}
    boundBone {boneName=dyn_shorts_out_r}
    chainEnd {boneName=dyn_shorts_out_02_r}
    physicsBodyDefinitions [{
      sphereCollisionRadius=0
      boundBone {boneName=dyn_shorts_out_r}
      boxExtents=<50 50 30>
      constraintSetup {angularTargetAxis=AxisZ, twistAxis=AxisZ, angularLimitsMax=<10 10 5>, angularLimitsMin=<-10 -10 -5>, angularTarget=<0 0 1>}
      localJointOffset=<0 0 6.5>
     }]
   }
   n34 = "Anim Dynamics - Chain: dyn_shorts_fr_r -> dyn_shorts_fr_02_r"(ComponentPose=@n33, Alpha=@n27, ExternalForce="0,0,0") {
    angularDampingOverride=0.999
    angularSpringConstant=300
    bAngularSpring=true
    bChain=true
    bGravityOverrideInSimSpace=true
    bOverrideAngularDamping=true
    numSolverIterationsPostUpdate=4
    numSolverIterationsPreUpdate=16
    simulationSpace=RootRelative
    alphaBoolBlend {blendOption=Linear}
    boundBone {boneName=dyn_shorts_fr_r}
    chainEnd {boneName=dyn_shorts_fr_02_r}
    physicsBodyDefinitions [{
      sphereCollisionRadius=0
      boundBone {boneName=dyn_shorts_fr_r}
      boxExtents=<26 22 8>
      constraintSetup {angularTargetAxis=AxisZ, twistAxis=AxisZ, angularLimitsMax=<10 0 5>, angularLimitsMin=<-10 -20 -5>, angularTarget=<0 0 1>}
      localJointOffset=<0 0 5>
     }]
   }
   n35 = "Anim Dynamics - Chain: dyn_shorts_bk_r -> dyn_shorts_bk_02_r"(ComponentPose=@n34, Alpha=@n27, ExternalForce="0,0,0") {
    angularDampingOverride=0.999
    angularSpringConstant=300
    bAngularSpring=true
    bChain=true
    bGravityOverrideInSimSpace=true
    bOverrideAngularDamping=true
    numSolverIterationsPostUpdate=4
    numSolverIterationsPreUpdate=16
    simulationSpace=RootRelative
    alphaBoolBlend {blendOption=Linear}
    boundBone {boneName=dyn_shorts_bk_r}
    chainEnd {boneName=dyn_shorts_bk_02_r}
    physicsBodyDefinitions [{
      sphereCollisionRadius=0
      boundBone {boneName=dyn_shorts_bk_r}
      boxExtents=<40 36 15>
      constraintSetup {angularTargetAxis=AxisZ, twistAxis=AxisZ, angularLimitsMax=<0 0 5>, angularLimitsMin=<0 -20 -5>, angularTarget=<0 0 1>}
      localJointOffset=<0 0 3.5>
     }]
   }
   n36 = Get calf_rot_back_l() -> (calf_rot_back_l:double)
   n37 = "Transform (Modify) Bone - Bone: dyn_shorts_main_l"(ComponentPose=@n35, Alpha=@n36, Translation="0.000,-0.000,-4", Rotation="0,0,-15", Scale="1,1,1") {
    rotationMode=BMM_Additive
    rotationSpace=BCS_BoneSpace
    scaleMode=BMM_Additive
    scaleSpace=BCS_BoneSpace
    translationMode=BMM_Additive
    translationSpace=BCS_BoneSpace
    alphaBoolBlend {blendOption=Linear}
    alphaScaleBiasClamp {bInterpResult=true, interpSpeedDecreasing=25, interpSpeedIncreasing=25}
    boneToModify {boneName=dyn_shorts_main_l}
    rotation {roll=-15}
    translation {z=-4}
   }
   n38 = "Transform (Modify) Bone - Bone: dyn_shorts_bk_l"(ComponentPose=@n37, Alpha=@n36, Translation="-0.247,0.000,22", Rotation="24.730,-3.101,-11.925", Scale="0.9,0.9,0.9") {
    rotationMode=BMM_Additive
    rotationSpace=BCS_BoneSpace
    scaleMode=BMM_Additive
    scaleSpace=BCS_BoneSpace
    translationMode=BMM_Additive
    translationSpace=BCS_BoneSpace
    alphaBoolBlend {blendOption=Linear}
    alphaScaleBiasClamp {bInterpResult=true, interpSpeedDecreasing=25, interpSpeedIncreasing=25}
    boneToModify {boneName=dyn_shorts_bk_l}
    rotation=<rot 24.73 -3.101 -11.925>
    scale=<0.9 0.9 0.9>
    translation {x=-0.247, z=22}
   }
   n39 = "Transform (Modify) Bone - Bone: dyn_shorts_out_l"(ComponentPose=@n38, Alpha=@n36, Translation="-0.000,0.022,7", Rotation="8,0.000,0.000", Scale="1,1,1") {
    rotationMode=BMM_Additive
    rotationSpace=BCS_BoneSpace
    scaleMode=BMM_Additive
    scaleSpace=BCS_BoneSpace
    translationMode=BMM_Additive
    translationSpace=BCS_BoneSpace
    alphaBoolBlend {blendOption=Linear}
    alphaScaleBiasClamp {bInterpResult=true, interpSpeedDecreasing=25, interpSpeedIncreasing=25}
    boneToModify {boneName=dyn_shorts_out_l}
    rotation {pitch=8}
    translation {y=0.022, z=7}
   }
   n40 = "Transform (Modify) Bone - Bone: dyn_shorts_in_l"(ComponentPose=@n39, Alpha=@n36, Translation="0.000,-0.033,15", Rotation="-12,0.000,0.000", Scale="1,1,1") {
    rotationMode=BMM_Additive
    rotationSpace=BCS_BoneSpace
    scaleMode=BMM_Additive
    scaleSpace=BCS_BoneSpace
    translationMode=BMM_Additive
    translationSpace=BCS_BoneSpace
    alphaBoolBlend {blendOption=Linear}
    alphaScaleBiasClamp {bInterpResult=true, interpSpeedDecreasing=25, interpSpeedIncreasing=25}
    boneToModify {boneName=dyn_shorts_in_l}
    rotation {pitch=-12}
    translation {y=-0.033, z=15}
   }
   n41 = "Transform (Modify) Bone - Bone: dyn_shorts_fr_l"(ComponentPose=@n40, Alpha=@n36, Translation="-9,0.400,0", Rotation="-10,0,0", Scale="1,1,1") {
    rotationMode=BMM_Additive
    rotationSpace=BCS_BoneSpace
    scaleMode=BMM_Additive
    scaleSpace=BCS_BoneSpace
    translationMode=BMM_Additive
    translationSpace=BCS_BoneSpace
    alphaBoolBlend {blendOption=Linear}
    alphaScaleBiasClamp {bInterpResult=true, interpSpeedDecreasing=25, interpSpeedIncreasing=25}
    boneToModify {boneName=dyn_shorts_fr_l}
    rotation {pitch=-10}
    translation=<-9 0.4>
   }
   n42 = "Transform (Modify) Bone - Bone: dyn_shorts_main_r"(ComponentPose=@n41, Alpha=@n1, Translation="0.000,-0.000,3", Rotation="0,0,10", Scale="1,1,1") {
    rotationMode=BMM_Additive
    rotationSpace=BCS_BoneSpace
    scaleMode=BMM_Additive
    scaleSpace=BCS_BoneSpace
    translationMode=BMM_Additive
    translationSpace=BCS_BoneSpace
    alphaBoolBlend {blendOption=Linear}
    alphaScaleBiasClamp {bInterpResult=true, interpSpeedDecreasing=25, interpSpeedIncreasing=25, scale=0}
    boneToModify {boneName=dyn_shorts_main_r}
    rotation {roll=10}
    translation {z=3}
   }
   n43 = "Transform (Modify) Bone - Bone: dyn_shorts_bk_r"(ComponentPose=@n42, Alpha=@n1, Translation="-0.247,0.000,-20", Rotation="24.730,-3.101,-11.925", Scale="0.9,0.9,0.9") {
    rotationMode=BMM_Additive
    rotationSpace=BCS_BoneSpace
    scaleMode=BMM_Additive
    scaleSpace=BCS_BoneSpace
    translationMode=BMM_Additive
    translationSpace=BCS_BoneSpace
    alphaBoolBlend {blendOption=Linear}
    alphaScaleBiasClamp {bInterpResult=true, interpSpeedDecreasing=25, interpSpeedIncreasing=25}
    boneToModify {boneName=dyn_shorts_bk_r}
    rotation=<rot 24.73 -3.101 -11.925>
    scale=<0.9 0.9 0.9>
    translation {x=-0.247, z=-20}
   }
   n44 = "Transform (Modify) Bone - Bone: dyn_shorts_out_r"(ComponentPose=@n43, Alpha=@n1, Translation="-0.000,0.022,-7", Rotation="8,0.000,0.000", Scale="1,1,1") {
    rotationMode=BMM_Additive
    rotationSpace=BCS_BoneSpace
    scaleMode=BMM_Additive
    scaleSpace=BCS_BoneSpace
    translationMode=BMM_Additive
    translationSpace=BCS_BoneSpace
    alphaBoolBlend {blendOption=Linear}
    alphaScaleBiasClamp {bInterpResult=true, interpSpeedDecreasing=25, interpSpeedIncreasing=25}
    boneToModify {boneName=dyn_shorts_out_r}
    rotation {pitch=8}
    translation {y=0.022, z=-7}
   }
   n45 = "Transform (Modify) Bone - Bone: dyn_shorts_in_r"(ComponentPose=@n44, Alpha=@n1, Translation="0.000,-0.033,-11", Rotation="0,0,0", Scale="1,1,1") {
    rotationMode=BMM_Additive
    rotationSpace=BCS_BoneSpace
    scaleMode=BMM_Additive
    scaleSpace=BCS_BoneSpace
    translationMode=BMM_Additive
    translationSpace=BCS_BoneSpace
    alphaBoolBlend {blendOption=Linear}
    alphaScaleBiasClamp {bInterpResult=true, interpSpeedDecreasing=25, interpSpeedIncreasing=25}
    boneToModify {boneName=dyn_shorts_in_r}
    translation {y=-0.033, z=-11}
   }
   n46 = "Transform (Modify) Bone - Bone: dyn_shorts_fr_r"(ComponentPose=@n45, Alpha=@n1, Translation="7,0.400,0.109", Rotation="-8,0,0", Scale="1,1,1") {
    rotationMode=BMM_Additive
    rotationSpace=BCS_BoneSpace
    scaleMode=BMM_Additive
    scaleSpace=BCS_BoneSpace
    translationMode=BMM_Additive
    translationSpace=BCS_BoneSpace
    alphaBoolBlend {blendOption=Linear}
    alphaScaleBiasClamp {bInterpResult=true, interpSpeedDecreasing=25, interpSpeedIncreasing=25, scale=0.5}
    boneToModify {boneName=dyn_shorts_fr_r}
    rotation {pitch=-8}
    translation=<7 0.4 0.109>
   }
   Output Pose(Result=Component To Local(ComponentPose=@n46)) {layerGroup=None}
   n47 = "Anim Dynamics - Chain: dyn_shorts_in_l -> dyn_shorts_in_02_l"!(ComponentPose, ExternalForce="0,0,0") -> (Pose) {
    angularDampingOverride=0.999
    angularSpringConstant=100
    bAngularSpring=true
    bChain=true
    bGravityOverrideInSimSpace=true
    bOverrideAngularDamping=true
    numSolverIterationsPostUpdate=4
    numSolverIterationsPreUpdate=16
    simulationSpace=Actor
    alphaBoolBlend {blendOption=Linear}
    boundBone {boneName=dyn_shorts_in_l}
    chainEnd {boneName=dyn_shorts_in_02_l}
    physicsBodyDefinitions [{
      sphereCollisionRadius=0
      boundBone {boneName=dyn_shorts_in_l}
      boxExtents=<25 30 12>
      constraintSetup {angularTargetAxis=AxisZ, twistAxis=AxisZ, angularLimitsMax=<10 10 5>, angularLimitsMin=<-10 -10 -5>, angularTarget=<0 0 1>}
      localJointOffset=<0 0 -3.5>
     }]
   }
   n48 = "Anim Dynamics - Chain: dyn_shorts_out_l -> dyn_shorts_out_02_l"!(ComponentPose=@n47, ExternalForce="0,0,0") -> (Pose) {
    angularDampingOverride=0.999
    angularSpringConstant=100
    bAngularSpring=true
    bChain=true
    bGravityOverrideInSimSpace=true
    bOverrideAngularDamping=true
    numSolverIterationsPostUpdate=4
    numSolverIterationsPreUpdate=16
    simulationSpace=Actor
    alphaBoolBlend {blendOption=Linear}
    boundBone {boneName=dyn_shorts_out_l}
    chainEnd {boneName=dyn_shorts_out_02_l}
    physicsBodyDefinitions [{
      sphereCollisionRadius=0
      boundBone {boneName=dyn_shorts_out_l}
      boxExtents=<50 50 30>
      constraintSetup {angularTargetAxis=AxisZ, twistAxis=AxisZ, angularLimitsMax=<10 10 5>, angularLimitsMin=<-10 -10 -5>, angularTarget=<0 0 1>}
      localJointOffset=<0 0 -2>
     }]
   }
   n49 = "Anim Dynamics - Chain: dyn_shorts_fr_l -> dyn_shorts_fr_02_l"!(ComponentPose=@n48, ExternalForce="0,0,0") -> (Pose) {
    angularDampingOverride=0.999
    angularSpringConstant=300
    bAngularSpring=true
    bChain=true
    bGravityOverrideInSimSpace=true
    bOverrideAngularDamping=true
    numSolverIterationsPostUpdate=4
    numSolverIterationsPreUpdate=16
    simulationSpace=Actor
    alphaBoolBlend {blendOption=Linear}
    boundBone {boneName=dyn_shorts_fr_l}
    chainEnd {boneName=dyn_shorts_fr_02_l}
    physicsBodyDefinitions [{
      sphereCollisionRadius=0
      boundBone {boneName=dyn_shorts_fr_l}
      boxExtents=<18 25 8>
      constraintSetup {angularTargetAxis=AxisZ, twistAxis=AxisZ, angularLimitsMax=<10 0 5>, angularLimitsMin=<-10 -20 -5>, angularTarget=<0 0 1>}
      localJointOffset=<0 0 -5>
     }]
   }
   n50 = "Anim Dynamics - Chain: dyn_shorts_bk_l -> dyn_shorts_bk_02_l"!(ComponentPose=@n49, ExternalForce="0,0,0") -> (Pose) {
    angularDampingOverride=0.999
    angularSpringConstant=300
    bAngularSpring=true
    bChain=true
    bGravityOverrideInSimSpace=true
    bOverrideAngularDamping=true
    numSolverIterationsPostUpdate=4
    numSolverIterationsPreUpdate=16
    simulationSpace=Actor
    alphaBoolBlend {blendOption=Linear}
    boundBone {boneName=dyn_shorts_bk_l}
    chainEnd {boneName=dyn_shorts_bk_02_l}
    physicsBodyDefinitions [{
      sphereCollisionRadius=0
      boundBone {boneName=dyn_shorts_bk_l}
      boxExtents=<40 38 15>
      constraintSetup {angularTargetAxis=AxisZ, twistAxis=AxisZ, angularLimitsMax=<10 0 5>, angularLimitsMin=<-10 -20 -5>, angularTarget=<0 0 1>}
      localJointOffset=<0 0 -3.5>
     }]
   }
   n51 = "Anim Dynamics - Chain: dyn_shorts_in_r -> dyn_shorts_in_02_r"!(ComponentPose=@n50, ExternalForce="0,0,0") -> (Pose) {
    angularDampingOverride=0.999
    angularSpringConstant=100
    bAngularSpring=true
    bChain=true
    bGravityOverrideInSimSpace=true
    bOverrideAngularDamping=true
    numSolverIterationsPostUpdate=4
    numSolverIterationsPreUpdate=16
    simulationSpace=Actor
    alphaBoolBlend {blendOption=Linear}
    boundBone {boneName=dyn_shorts_in_r}
    chainEnd {boneName=dyn_shorts_in_02_r}
    physicsBodyDefinitions [{
      sphereCollisionRadius=0
      boundBone {boneName=dyn_shorts_in_r}
      boxExtents=<25 29 12>
      constraintSetup {angularTargetAxis=AxisZ, twistAxis=AxisZ, angularLimitsMax=<10 10 5>, angularLimitsMin=<-10 -10 -5>, angularTarget=<0 0 1>}
      localJointOffset=<0 0 3.5>
     }]
   }
   n52 = "Anim Dynamics - Chain: dyn_shorts_out_r -> dyn_shorts_out_02_r"!(ComponentPose=@n51, ExternalForce="0,0,0") -> (Pose) {
    angularDampingOverride=0.999
    angularSpringConstant=100
    bAngularSpring=true
    bChain=true
    bGravityOverrideInSimSpace=true
    bOverrideAngularDamping=true
    numSolverIterationsPostUpdate=4
    numSolverIterationsPreUpdate=16
    simulationSpace=Actor
    alphaBoolBlend {blendOption=Linear}
    boundBone {boneName=dyn_shorts_out_r}
    chainEnd {boneName=dyn_shorts_out_02_r}
    physicsBodyDefinitions [{
      sphereCollisionRadius=0
      boundBone {boneName=dyn_shorts_out_r}
      boxExtents=<50 50 30>
      constraintSetup {angularTargetAxis=AxisZ, twistAxis=AxisZ, angularLimitsMax=<10 10 5>, angularLimitsMin=<-10 -10 -5>, angularTarget=<0 0 1>}
      localJointOffset=<0 0 2>
     }]
   }
   n53 = "Anim Dynamics - Chain: dyn_shorts_fr_r -> dyn_shorts_fr_02_r"!(ComponentPose=@n52, ExternalForce="0,0,0") -> (Pose) {
    angularDampingOverride=0.999
    angularSpringConstant=300
    bAngularSpring=true
    bChain=true
    bGravityOverrideInSimSpace=true
    bOverrideAngularDamping=true
    numSolverIterationsPostUpdate=4
    numSolverIterationsPreUpdate=16
    simulationSpace=Actor
    alphaBoolBlend {blendOption=Linear}
    boundBone {boneName=dyn_shorts_fr_r}
    chainEnd {boneName=dyn_shorts_fr_02_r}
    physicsBodyDefinitions [{
      sphereCollisionRadius=0
      boundBone {boneName=dyn_shorts_fr_r}
      boxExtents=<26 22 8>
      constraintSetup {angularTargetAxis=AxisZ, twistAxis=AxisZ, angularLimitsMax=<10 0 5>, angularLimitsMin=<-10 -20 -5>, angularTarget=<0 0 1>}
      localJointOffset=<0 0 5>
     }]
   }
   n54 = "Anim Dynamics - Chain: dyn_shorts_bk_r -> dyn_shorts_bk_02_r"!(ComponentPose=@n53, ExternalForce="0,0,0") -> (Pose) {
    angularDampingOverride=0.999
    angularSpringConstant=300
    bAngularSpring=true
    bChain=true
    bGravityOverrideInSimSpace=true
    bOverrideAngularDamping=true
    numSolverIterationsPostUpdate=4
    numSolverIterationsPreUpdate=16
    simulationSpace=Actor
    alphaBoolBlend {blendOption=Linear}
    boundBone {boneName=dyn_shorts_bk_r}
    chainEnd {boneName=dyn_shorts_bk_02_r}
    physicsBodyDefinitions [{
      sphereCollisionRadius=0
      boundBone {boneName=dyn_shorts_bk_r}
      boxExtents=<40 36 15>
      constraintSetup {angularTargetAxis=AxisZ, twistAxis=AxisZ, angularLimitsMax=<0 0 5>, angularLimitsMin=<0 -20 -5>, angularTarget=<0 0 1>}
      localJointOffset=<0 0 3.5>
     }]
   }
   n55 = Flap Node - Bone: dyn_shorts_bk_02_l!(ComponentPose:Component Space Pose Link Structure=@n54, Alpha:float=1) -> (Pose:Component Space Pose Link Structure) {
    is=title FortAnimGraphNode_Flap
    targetComponent=RotationY
    alphaBoolBlend {blendOption=Linear}
    sourceDrives [{
      bUseRange=true
      drivingCurve=null
      multiplier=1
      rangeMax=-0.53
      rangeMin=-2.094
      remappedMax=0
      remappedMin=-1.919
      sourceComponent=RotationZ
      useQuaternion=true
      sourceBone {boneName=calf_l}
     }]
    targetBone {boneName=dyn_shorts_bk_02_l}
   }
   n56 = Flap Node - Bone: dyn_shorts_bk_02_l(ComponentPose=@n55, Alpha=1) {
    targetComponent=ScaleZ
    alphaBoolBlend {blendOption=Linear}
    sourceDrives [{
      bUseRange=true
      drivingCurve=null
      multiplier=1
      rangeMax=-0.53
      rangeMin=-2.094
      remappedMax=0
      remappedMin=-0.8
      sourceComponent=RotationZ
      useQuaternion=true
      sourceBone {boneName=calf_l}
     }]
    targetBone {boneName=dyn_shorts_bk_02_l}
   }
   n57 = Flap Node - Bone: dyn_shorts_bk_02_r!(ComponentPose:Component Space Pose Link Structure=@n56, Alpha:float=1) -> (Pose:Component Space Pose Link Structure) {
    is=title FortAnimGraphNode_Flap
    targetComponent=RotationY
    alphaBoolBlend {blendOption=Linear}
    sourceDrives [{
      bUseRange=true
      drivingCurve=null
      multiplier=1
      rangeMax=-0.53
      rangeMin=-2.094
      remappedMax=0
      remappedMin=-1.99
      sourceComponent=RotationZ
      useQuaternion=true
      sourceBone {boneName=calf_r}
     }]
    targetBone {boneName=dyn_shorts_bk_02_r}
   }
   Flap Node - Bone: dyn_shorts_bk_02_r(ComponentPose=@n57, Alpha=1) {
    targetComponent=ScaleZ
    alphaBoolBlend {blendOption=Linear}
    sourceDrives [{
      bUseRange=true
      drivingCurve=null
      multiplier=1
      rangeMax=-0.53
      rangeMin=-2.094
      remappedMax=0
      remappedMin=-0.9
      sourceComponent=RotationZ
      useQuaternion=true
      sourceBone {boneName=calf_r}
     }]
    targetBone {boneName=dyn_shorts_bk_02_r}
   }
  ]}
 EventGraph {nodes [
   Event BlueprintUpdateAnimation()
   > n1 = Sequence!(execute) -> (then_0, then_1, then_2:Exec, then_3:Exec)
   n2 = BreakTransform(InTransform=GetDeltaTransformFromRefPose(self=GetOwnerSkeletalMeshComponent(), BoneName=calf_l, BaseName=thigh_l))
   Set calf_rot_back_l(execute=@n1.then_1, calf_rot_back_l:double=MapRangeClamped(Value=BreakRotator(InRot=@n2.Rotation).Yaw, InRangeA=15, InRangeB=115)) -> (Output_Get:double)
   n3 = BreakTransform(InTransform=GetDeltaTransformFromRefPose(self=GetOwnerSkeletalMeshComponent(), BoneName=calf_r, BaseName=thigh_r))
   Set calf_rot_back_r(execute=@n1.then_0, calf_rot_back_r:double=MapRangeClamped(Value=BreakRotator(InRot=@n3.Rotation).Yaw, InRangeA=15, InRangeB=115)) -> (Output_Get:double)
   Event OnBeginCrouching()
   > Set crouch_alpha(crouch_alpha:double=1) -> (Output_Get:double)
   Event OnEndCrouching()
   > Set crouch_alpha(crouch_alpha=0)
   TryGetPawnOwner()
  ]}
]
variables [
 calf_rot_back_l {type=number}
 calf_rot_back_r {type=number}
 crouch_alpha {type=number}
 Shorts_Alpha {type=number, default=0.6}
]
